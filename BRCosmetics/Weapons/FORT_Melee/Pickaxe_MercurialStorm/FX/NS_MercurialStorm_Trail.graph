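# Niagara system for the Mercurial Storm pickaxe trail, with a single AirTrail ribbon emitter built on the shared AnimTrail spawn, initialize and update modules.
# Ribbon color lerps from User.TrailColor to a hue-shifted, more saturated, much brighter version of it, driven by User.Glow.
# Spawn alpha drops to zero early in the emitter's age, and each short-lived particle also fades its alpha over life.
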NiagaraSystem /BRCosmetics/Weapons/FORT_Melee/Pickaxe_MercurialStorm/FX/NS_MercurialStorm_Trail
dependencies {usedDataInterfaces [&NiagaraDataInterfaceCurve], usedDynamicInputs [], usedModules [&SystemState]}
emitters {AirTrail {
  emitterUpdateScript [
   &EmitterState()
   &FN_SpawnAnimTrail_v2(Spawn Interval=0)
  ]
  particleSpawnScript [
   /Game/Effects/Niagara/Modules/AnimTrail/FN_InitializeAnimTrail/FN_InitializeAnimTrailBoneIndecies(First Bone or Socket Index=1)
   /Game/Effects/Niagara/Modules/AnimTrail/FN_InitializeAnimTrail/FN_InitializeAnimTrail(
    AnimTrailWidth=1.25
    Color=&MultiplyLinearColor(
     Linear Color=&LerpLinearColors(
      StartColor=$User.TrailColor
      EndColor=&HueShiftLinearColor(Hue=-0.05, Saturation=5, Value=10, Linear Color=$User.TrailColor)
      LerpFactor=$User.Glow
     )
     Scale Factor=&MakeFloat4FromVectorAndFloat(
      "Vector (RGB)"=<1 1 1>
      "Float (Alpha)"=&FloatFromCurve(FloatCurve=DataInterface() {propertyValues {Curve {keys [<at 0.07 1>,<at 0.125 0>]}}}, CurveIndex=$Emitter.Age)
     )
    )
    Lifetime=0.125
   )
  ]
  particleUpdateScript [
   &SolveForcesAndVelocity(Clamp Velocity=false, Limit Acceleration=false)
   &FN_UpdateAnimTrail_v2()
   &ScaleColor(
    Scale Alpha=&FloatFromCurve(FloatCurve=DataInterface() {propertyValues {Curve {keys [<at 0 1 RCIM_Cubic>,<at 1 0>]}}})
    ScaleA=true
    ScaleRGB=true
   )
  ]
  rendererClasses [&NiagaraRibbonRendererProperties]
  renderers [{rendererClass=&NiagaraRibbonRendererProperties}]
 }}
summary {
 systemName=NS_MercurialStorm_Trail
 emitters [{emitterName=AirTrail, rendererClasses [&NiagaraRibbonRendererProperties]}]
 userVariables [
  User.Glow {type=NiagaraFloat, defaultValue=0, description=""}
  User.RibbonIDOffset {type=NiagaraInt32}
  User.SkeletalMesh {
   type=NiagaraDataInterfaceSkeletalMesh
   defaultValue {
    struct=NiagaraExt_VariableValue_DataInterface
    value {dataInterface=&self:NiagaraDataInterfaceSkeletalMesh_3}
   }
  }
  User.TrailColor {type=LinearColor, description="", defaultValue {struct=LinearColor, value=<rgba 0.5278 0.5278 0.5278 1>}}
 ]
}
